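# Physics asset for the female medium Witch player body's loincloth, defining collision and joints for its front flap.
# Three box bodies form a flap chain from the pelvis, joined by constraints with tight twist and side-swing limits and a wider swing on one axis.
# Capsules on the pelvis, lower spine, thighs and calves give the leg and hip shapes, linked by constraints with free swing and twist.
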PhysicsAsset /Game/Characters/Player/Female/Medium/Bodies/F_MED_Witch/Meshes/F_MED_Witch_Loincloth_Phys
bodies {
 calf_l [{shapeType=Capsule, length=48.6232, radius=6.0825, center=<-26.865 -2.4355 -0.0692>, rotation=<rot 90 0 0>}]
 calf_r [{shapeType=Capsule, length=48.6232, radius=6.0825, center=<26.865 2.4355 -0.0692>, rotation=<rot 90 0 180>}]
 dyn_flap_front_01 [{shapeType=Box, extentX=13.0613, extentY=2.7509, extentZ=9.8182, center=<6.689 -0.0511 0>}]
 dyn_flap_front_02 [{shapeType=Box, extentX=13.3439, extentY=0.9599, extentZ=9.718, center=<6.7525 -0.1983 0.0001>}]
 dyn_flap_front_03 [{shapeType=Box, extentX=13.5771, extentY=2.2354, extentZ=11.1986, center=<6.2618 -0.4141 0>}]
 pelvis [{shapeType=Capsule, length=15.7124, radius=7.3885, center=<-4.0908 -0.9992 0.6618>, rotation=<rot -1.5272 -88.8307 0.3619>}]
 spine_01 [{shapeType=Capsule, length=0.505, radius=0.505, rotation=<rot 0 -90 0>}]
 thigh_l [{shapeType=Capsule, length=26.4371, radius=7.1084, center=<-22.2289 0.0138 -0.1292>, rotation=<rot -0.2108 91.9184 -92.0558>}]
 thigh_r [{shapeType=Capsule, length=25.5185, radius=7.6339, center=<22.2289 -0.0138 -0.1292>, rotation=<rot -0.2108 -88.0816 -92.0558>}]
}
constraints [
 {
  bone1Name=calf_l
  bone2Name=thigh_l
  swing1LimitDegrees=45
  swing1Motion=Free
  swing2LimitDegrees=45
  swing2Motion=Free
  twistLimitDegrees=45
  twistMotion=Free
 }
 {
  bone1Name=calf_r
  bone2Name=thigh_r
  swing1LimitDegrees=45
  swing1Motion=Free
  swing2LimitDegrees=45
  swing2Motion=Free
  twistLimitDegrees=45
  twistMotion=Free
 }
 {bone1Name=dyn_flap_front_02, bone2Name=dyn_flap_front_01, swing1LimitDegrees=20, swing2LimitDegrees=5, twistLimitDegrees=1}
 {bone1Name=dyn_flap_front_03, bone2Name=dyn_flap_front_02, swing1LimitDegrees=20, swing2LimitDegrees=5, twistLimitDegrees=1}
 {
  bone1Name=spine_01
  bone2Name=pelvis
  swing1LimitDegrees=45
  swing1Motion=Free
  swing2LimitDegrees=45
  swing2Motion=Free
  twistLimitDegrees=45
  twistMotion=Free
 }
 {
  bone1Name=thigh_l
  bone2Name=pelvis
  swing1LimitDegrees=45
  swing1Motion=Free
  swing2LimitDegrees=45
  swing2Motion=Free
  twistLimitDegrees=45
  twistMotion=Free
 }
 {
  bone1Name=thigh_r
  bone2Name=pelvis
  swing1LimitDegrees=45
  swing1Motion=Free
  swing2LimitDegrees=45
  swing2Motion=Free
  twistLimitDegrees=45
  twistMotion=Free
 }
 {bone1Name=dyn_flap_front_01, bone2Name=pelvis, swing1LimitDegrees=20, swing2LimitDegrees=5, twistLimitDegrees=1}
]
