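# Physics asset for the female medium Witch player body, covering dangling accessories and a few core bones rather than a full ragdoll.
# The book hangs from the pelvis with its flap chained to it, while scroll, potion and waist sack hang from a tiny spine_01 capsule.
# Boot buckles and leg cloth attach to each calf with limited angular motion.
# Calves and spine_01 link to the pelvis with free swing and twist.
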PhysicsAsset /Game/Characters/Player/Female/Medium/Bodies/F_MED_Witch/Meshes/F_MED_Witch_Physics
bodies {
 calf_l [{shapeType=Box, extentX=16.4973, extentY=12.165, extentZ=10.6043, center=<-30.8967 -2.4358 -0.0701>}]
 calf_r [{shapeType=Box, extentX=16.0472, extentY=12.2613, extentZ=10.5422, center=<31.7661 2.5693 0.0942>}]
 dyn_book [{shapeType=Box, extentX=22.6034, extentY=3.5454, extentZ=16.7601, center=<1.0931 0.3994 -7.9988>, rotation=<rot -9.3549 -87.0311 -17.4804>}]
 dyn_book_flap [{shapeType=Box, extentX=1.6001, extentY=16.451, extentZ=24.5742, center=<5.831 -0.2147 -6.5732>, rotation=<rot -29.3068 178.9346 -79.8423>}]
 dyn_boot_buckle_l [{shapeType=Box, extentX=4.6797, extentY=2.8773, extentZ=4.4382, center=<3.8934 -2.3334 0.3158>}]
 dyn_boot_buckle_r [{shapeType=Box, extentX=4.6797, extentY=2.8773, extentZ=4.4382, center=<-3.8934 -2.3334 0.3158>}]
 dyn_cloth_leg_l [{shapeType=Sphere, radius=9.301, center=<0.4405 -0.5177 -5.2834>}]
 dyn_cloth_leg_r [{shapeType=Sphere, radius=9.301, center=<0.4405 -0.5177 -5.2834>}]
 dyn_potion [{shapeType=Capsule, length=6.1266, radius=3.8733, center=<-0.9271 0.1864 -1.9782>, rotation=<rot -0.3036 -89.9996 14.1788>}]
 dyn_scroll [{shapeType=Capsule, length=8.4242, radius=2.1965, center=<-0.3094 0.0282 -1.0957>, rotation=<rot -1.0814 -90 15.5017>}]
 dyn_waist_sack [{shapeType=Capsule, length=0.0199, radius=4.2417, center=<-2.7092 -1.3611 -5.073>, rotation=<rot 8.7192 -90 -7.1339>}]
 pelvis [{shapeType=Capsule, length=10.5106, radius=6.6862, center=<-0.1284 -0.4824 0.1902>, rotation=<rot 1.7162 -87.6336 0.8119>}]
 spine_01 [{shapeType=Capsule, length=0.505, radius=0.505, rotation=<rot 0 -90 0>}]
}
constraints [
 {bone1Name=dyn_book, bone2Name=pelvis, swing1LimitDegrees=10, swing2LimitDegrees=15, twistLimitDegrees=45}
 {bone1Name=dyn_book_flap, bone2Name=dyn_book, swing1LimitDegrees=1, swing2LimitDegrees=1, twistLimitDegrees=30}
 {bone1Name=dyn_scroll, bone2Name=spine_01, swing1LimitDegrees=1, swing2LimitDegrees=1, twistLimitDegrees=20}
 {bone1Name=dyn_potion, bone2Name=spine_01, swing1LimitDegrees=20, swing2LimitDegrees=20, twistLimitDegrees=1}
 {bone1Name=dyn_waist_sack, bone2Name=spine_01, swing1LimitDegrees=20, swing2LimitDegrees=20, twistLimitDegrees=1}
 {
  bone1Name=spine_01
  bone2Name=pelvis
  swing1LimitDegrees=45
  swing1Motion=Free
  swing2LimitDegrees=45
  swing2Motion=Free
  twistLimitDegrees=45
  twistMotion=Free
 }
 {bone1Name=dyn_boot_buckle_l, bone2Name=calf_l, swing1LimitDegrees=1, swing2LimitDegrees=1, twistLimitDegrees=20}
 {bone1Name=dyn_boot_buckle_r, bone2Name=calf_r, swing1LimitDegrees=1, swing2LimitDegrees=1, twistLimitDegrees=20}
 {
  bone1Name=calf_r
  bone2Name=pelvis
  swing1LimitDegrees=45
  swing1Motion=Free
  swing2LimitDegrees=45
  swing2Motion=Free
  twistLimitDegrees=45
  twistMotion=Free
 }
 {
  bone1Name=calf_l
  bone2Name=pelvis
  swing1LimitDegrees=45
  swing1Motion=Free
  swing2LimitDegrees=45
  swing2Motion=Free
  twistLimitDegrees=45
  twistMotion=Free
 }
 {bone1Name=dyn_cloth_leg_r, bone2Name=calf_r, swing1LimitDegrees=10, swing2LimitDegrees=10, twistLimitDegrees=1}
 {bone1Name=dyn_cloth_leg_l, bone2Name=calf_l, swing1LimitDegrees=10, swing2LimitDegrees=10, twistLimitDegrees=1}
]
