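# Physics asset for the medium female Witch's skirt, defining bodies and joints for its dynamic dress bones and back flap.
# Capsules cover spine, pelvis, thighs and calves with free-swinging joints, while boxes form three-segment front, side and back chains per leg.
# Each chain roots on its thigh rather than the pelvis, with tight angular limits; the flap hangs from the pelvis.
# Cross-constraints tie lower front segments to side segments, leaving back chains unlinked.
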PhysicsAsset /Game/Characters/Player/Female/Medium/Bodies/F_MED_Witch/Meshes/F_MED_Witch_Skirt_Phys
bodies {
 calf_l [{shapeType=Capsule, length=50.1301, radius=6.4313, center=<-24.0605 -2.4219 -0.1549>, rotation=<rot 90 0 0>}]
 calf_r [{shapeType=Capsule, length=50.7921, radius=6.6218, center=<23.6737 2.9261 0.4933>, rotation=<rot -1.4369 90.8296 -88.6169>}]
 dyn_dress_back_l_01 [{shapeType=Box, extentX=12.6771, extentY=2.4696, extentZ=12.7618, center=<7.0912 0.804 0.3783>, rotation=<rot -1.965 5.1655 -164.8097>}]
 dyn_dress_back_l_02 [{shapeType=Box, extentX=14.2148, extentY=4.525, extentZ=22.5165, center=<7.3931 0.4774 0.1128>, rotation=<rot -1.1291 0.3863 -161.6983>}]
 dyn_dress_back_l_03 [{shapeType=Box, extentX=15.5906, extentY=3.2695, extentZ=25.521, center=<7.7115 0.8477 -0.4635>, rotation=<rot -1.2242 0.8112 -162.6908>}]
 dyn_dress_back_r_01 [{shapeType=Box, extentX=12.6771, extentY=2.4696, extentZ=14.4026, center=<7.1472 1.0286 -1.0644>, rotation=<rot -10.5569 2.7527 -14.0077>}]
 dyn_dress_back_r_02 [{shapeType=Box, extentX=23.9322, extentY=3.6301, extentZ=18.7345, center=<8.8858 -0.0933 -2.0288>, rotation=<rot 69.2558 -64.012 -64.5697>}]
 dyn_dress_back_r_03 [{shapeType=Box, extentX=22.2077, extentY=15.9909, extentZ=4.0054, center=<8.6924 -0.0119 -0.2229>, rotation=<rot 75.8543 -53.9165 31.3622>}]
 dyn_dress_front_l_01 [{shapeType=Box, extentX=13.5196, extentY=2.4696, extentZ=12.7618, center=<7.5128 -0.6097 0.0012>, rotation=<rot -2.3229 -1.5421 163.1587>}]
 dyn_dress_front_l_02 [{shapeType=Box, extentX=14.9755, extentY=4.1409, extentZ=21.0879, center=<8.1321 0.0113 -0.0929>, rotation=<rot -1.2242 0.8112 159.4354>}]
 dyn_dress_front_l_03 [{shapeType=Box, extentX=17.341, extentY=4.3082, extentZ=21.5596, center=<8.8795 0.1698 -0.1761>, rotation=<rot -1.2242 0.8112 161.3205>}]
 dyn_dress_front_r_01 [{shapeType=Box, extentX=12.6771, extentY=3.5537, extentZ=12.7618, center=<7.1503 -0.2959 -1.3891>, rotation=<rot -9.3775 0.6176 -164.5374>}]
 dyn_dress_front_r_02 [{shapeType=Box, extentX=14.3374, extentY=3.9857, extentZ=21.5142, center=<7.8846 -0.7822 -1.371>, rotation=<rot -9.3775 0.6176 -164.5371>}]
 dyn_dress_front_r_03 [{shapeType=Box, extentX=16.0965, extentY=4.0461, extentZ=22.0718, center=<7.9821 -0.2132 -1.5474>, rotation=<rot -9.3775 0.6176 -164.5371>}]
 dyn_dress_side_l_01 [{shapeType=Box, extentX=16.2026, extentY=4.4786, extentZ=14.0534, center=<5.0379 -0.4993 3.2781>, rotation=<rot -0.4548 0.9682 -87.7707>}]
 dyn_dress_side_l_02 [{shapeType=Box, extentX=14.3374, extentY=3.3239, extentZ=22.8803, center=<7.9554 -0.6121 -0.4157>, rotation=<rot -1.2242 0.8112 -88.0133>}]
 dyn_dress_side_l_03 [{shapeType=Box, extentX=14.3374, extentY=2.4696, extentZ=23.8479, center=<9.3042 -0.2121 -0.458>, rotation=<rot -1.2242 0.8112 -88.0133>}]
 dyn_dress_side_r_01 [{shapeType=Box, extentX=13.2449, extentY=3.465, extentZ=19.4831, center=<6.607 -0.2908 -1.2591>, rotation=<rot -13.4584 -0.5426 -93.7079>}]
 dyn_dress_side_r_02 [{shapeType=Box, extentX=26.1715, extentY=3.6257, extentZ=12.0826, center=<6.7915 -0.0072 -1.1111>, rotation=<rot -0.8229 -90.8688 -80.336>}]
 dyn_dress_side_r_03 [{shapeType=Box, extentX=22.8114, extentY=15.7458, extentZ=4.123, center=<7.8263 0.0003 -1.2608>, rotation=<rot 0 -89.9999 10.1737>}]
 dyn_flap_back [{shapeType=Box, extentX=1.643, extentY=10.4359, extentZ=14.0575, center=<4.4564 1.879 -0.0001>, rotation=<rot 0 -82.7951 0>}]
 pelvis [{shapeType=Capsule, length=10.5106, radius=6.6862, center=<-3.6279 -6.7898 0.428>, rotation=<rot 1.7162 -87.6336 0.8119>}]
 spine_01 [{shapeType=Capsule, length=0.505, radius=0.505, rotation=<rot 0 -90 0>}]
 thigh_l [{shapeType=Capsule, length=31.4683, radius=7.5478, center=<-22.2289 0.0104 -0.1354>, rotation=<rot -0.2107 91.9184 -90.0581>}]
 thigh_r [{shapeType=Capsule, length=30.9045, radius=7.4429, center=<22.2335 0.3634 -0.6029>, rotation=<rot -0.2108 -88.0816 -92.0558>}]
}
constraints [
 {
  bone1Name=spine_01
  bone2Name=pelvis
  swing1LimitDegrees=45
  swing1Motion=Free
  swing2LimitDegrees=45
  swing2Motion=Free
  twistLimitDegrees=45
  twistMotion=Free
 }
 {bone1Name=dyn_flap_back, bone2Name=pelvis, swing1LimitDegrees=1, swing2LimitDegrees=1, twistLimitDegrees=15}
 {bone1Name=dyn_dress_side_r_01, bone2Name=thigh_r, swing1LimitDegrees=7, swing2LimitDegrees=5, twistLimitDegrees=5}
 {bone1Name=dyn_dress_side_r_02, bone2Name=dyn_dress_side_r_01, swing1LimitDegrees=25, swing2LimitDegrees=20, twistLimitDegrees=5}
 {bone1Name=dyn_dress_side_r_03, bone2Name=dyn_dress_side_r_02, swing1LimitDegrees=25, swing2LimitDegrees=20, twistLimitDegrees=5}
 {
  bone1Name=thigh_r
  bone2Name=pelvis
  swing1LimitDegrees=45
  swing1Motion=Free
  swing2LimitDegrees=45
  swing2Motion=Free
  twistLimitDegrees=45
  twistMotion=Free
 }
 {bone1Name=dyn_dress_front_r_01, bone2Name=thigh_r, swing1LimitDegrees=5, swing2LimitDegrees=5, twistLimitDegrees=5}
 {bone1Name=dyn_dress_front_r_02, bone2Name=dyn_dress_front_r_01, swing1LimitDegrees=5, swing2LimitDegrees=20, twistLimitDegrees=5}
 {bone1Name=dyn_dress_front_r_03, bone2Name=dyn_dress_front_r_02, swing1LimitDegrees=5, swing2LimitDegrees=20, twistLimitDegrees=5}
 {bone1Name=dyn_dress_back_r_01, bone2Name=thigh_r, swing1LimitDegrees=25, swing2LimitDegrees=5, twistLimitDegrees=5}
 {bone1Name=dyn_dress_back_r_02, bone2Name=dyn_dress_back_r_01, swing1LimitDegrees=5, swing2LimitDegrees=20, twistLimitDegrees=5}
 {bone1Name=dyn_dress_back_r_03, bone2Name=dyn_dress_back_r_02, swing1LimitDegrees=25, swing2LimitDegrees=15, twistLimitDegrees=5}
 {bone1Name=dyn_dress_side_l_01, bone2Name=thigh_l, swing1LimitDegrees=5, swing2LimitDegrees=10, twistLimitDegrees=5}
 {bone1Name=dyn_dress_side_l_02, bone2Name=dyn_dress_side_l_01, swing1LimitDegrees=30, swing2LimitDegrees=20, twistLimitDegrees=5}
 {bone1Name=dyn_dress_side_l_03, bone2Name=dyn_dress_side_l_02, swing1LimitDegrees=25, swing2LimitDegrees=20, twistLimitDegrees=5}
 {bone1Name=dyn_dress_front_l_01, bone2Name=thigh_l, swing1LimitDegrees=5, swing2LimitDegrees=5, twistLimitDegrees=5}
 {bone1Name=dyn_dress_front_l_02, bone2Name=dyn_dress_front_l_01, swing1LimitDegrees=5, swing2LimitDegrees=20, twistLimitDegrees=5}
 {bone1Name=dyn_dress_front_l_03, bone2Name=dyn_dress_front_l_02, swing1LimitDegrees=5, swing2LimitDegrees=20, twistLimitDegrees=5}
 {bone1Name=dyn_dress_back_l_01, bone2Name=thigh_l, swing1LimitDegrees=20, swing2LimitDegrees=5, twistLimitDegrees=5}
 {bone1Name=dyn_dress_back_l_02, bone2Name=dyn_dress_back_l_01, swing1LimitDegrees=5, swing2LimitDegrees=20, twistLimitDegrees=5}
 {bone1Name=dyn_dress_back_l_03, bone2Name=dyn_dress_back_l_02, swing1LimitDegrees=5, swing2LimitDegrees=25, twistLimitDegrees=5}
 {
  bone1Name=thigh_l
  bone2Name=pelvis
  swing1LimitDegrees=45
  swing1Motion=Free
  swing2LimitDegrees=45
  swing2Motion=Free
  twistLimitDegrees=45
  twistMotion=Free
 }
 {
  bone1Name=calf_l
  bone2Name=thigh_l
  swing1LimitDegrees=45
  swing1Motion=Free
  swing2LimitDegrees=45
  swing2Motion=Free
  twistLimitDegrees=45
  twistMotion=Free
 }
 {bone1Name=dyn_dress_front_r_03, bone2Name=dyn_dress_side_r_03, swing1LimitDegrees=45, swing2LimitDegrees=45, twistLimitDegrees=1}
 {bone1Name=dyn_dress_front_r_02, bone2Name=dyn_dress_side_r_02, swing1LimitDegrees=45, swing2LimitDegrees=45, twistLimitDegrees=1}
 {bone1Name=dyn_dress_front_l_02, bone2Name=dyn_dress_side_l_02, swing1LimitDegrees=45, swing2LimitDegrees=45, twistLimitDegrees=1}
 {bone1Name=dyn_dress_front_l_03, bone2Name=dyn_dress_side_l_03, swing1LimitDegrees=45, swing2LimitDegrees=45, twistLimitDegrees=1}
 {
  bone1Name=calf_r
  bone2Name=thigh_r
  swing1LimitDegrees=45
  swing1Motion=Free
  swing2LimitDegrees=45
  swing2Motion=Free
  twistLimitDegrees=45
  twistMotion=Free
 }
]
