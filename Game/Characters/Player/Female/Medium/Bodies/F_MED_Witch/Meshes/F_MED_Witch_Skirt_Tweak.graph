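# Physics asset for the medium female Witch skirt, simulating the back dress panels against simplified leg bodies.
# Each side chains two box bodies, dyn_dress_back 02 and 03, from the thigh with tightly limited swing and twist.
# Thighs and left calf use capsules, the right calf a box, and calf-to-thigh constraints leave rotation free.
# A tiny spine_01 capsule has no constraint, and the lower panels' swing limits are swapped between sides.
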PhysicsAsset /Game/Characters/Player/Female/Medium/Bodies/F_MED_Witch/Meshes/F_MED_Witch_Skirt_Tweak
bodies {
 calf_l [{shapeType=Capsule, length=50.1301, radius=6.4313, center=<-24.0605 -2.4219 -0.1549>, rotation=<rot 90 0 0>}]
 calf_r [{shapeType=Box, extentX=13.2437, extentY=10.5692, extentZ=57.2044, center=<25.8212 0.4081 0.6843>, rotation=<rot 1.7202 91.1514 -88.3973>}]
 dyn_dress_back_l_02 [{shapeType=Box, extentX=14.2148, extentY=7.8255, extentZ=22.5165, center=<7.3931 0.4774 0.1128>, rotation=<rot -1.1291 0.3863 -161.6983>}]
 dyn_dress_back_l_03 [{shapeType=Box, extentX=27.4405, extentY=4.2868, extentZ=25.521, center=<12.0787 0.992 -0.1547>, rotation=<rot -1.2242 0.8112 -162.6908>}]
 dyn_dress_back_r_02 [{shapeType=Box, extentX=23.9322, extentY=6.1511, extentZ=18.7345, center=<8.8858 -0.0933 -2.0288>, rotation=<rot 69.2558 -64.012 -64.5697>}]
 dyn_dress_back_r_03 [{shapeType=Box, extentX=22.2077, extentY=28.8055, extentZ=5.0002, center=<13.3949 1.5272 -1.04>, rotation=<rot 75.8543 -53.9165 31.3623>}]
 spine_01 [{shapeType=Capsule, length=0.505, radius=0.505, rotation=<rot 0 -90 0>}]
 thigh_l [{shapeType=Capsule, length=31.4683, radius=7.5478, center=<-22.2289 0.0104 -0.1354>, rotation=<rot -0.2107 91.9184 -90.0581>}]
 thigh_r [{shapeType=Capsule, length=30.9045, radius=7.4429, center=<22.2335 0.3634 -0.6029>, rotation=<rot -0.2108 -88.0816 -92.0558>}]
}
constraints [
 {bone1Name=dyn_dress_back_r_02, bone2Name=thigh_r, swing1LimitDegrees=20, swing2LimitDegrees=1, twistLimitDegrees=1}
 {bone1Name=dyn_dress_back_r_03, bone2Name=dyn_dress_back_r_02, swing1LimitDegrees=30, swing2LimitDegrees=3, twistLimitDegrees=1}
 {bone1Name=dyn_dress_back_l_02, bone2Name=thigh_l, swing1LimitDegrees=20, swing2LimitDegrees=1, twistLimitDegrees=1}
 {bone1Name=dyn_dress_back_l_03, bone2Name=dyn_dress_back_l_02, swing1LimitDegrees=3, swing2LimitDegrees=30, twistLimitDegrees=1}
 {
  bone1Name=calf_l
  bone2Name=thigh_l
  swing1LimitDegrees=45
  swing1Motion=Free
  swing2LimitDegrees=45
  swing2Motion=Free
  twistLimitDegrees=45
  twistMotion=Free
 }
 {
  bone1Name=calf_r
  bone2Name=thigh_r
  swing1LimitDegrees=45
  swing1Motion=Free
  swing2LimitDegrees=45
  swing2Motion=Free
  twistLimitDegrees=45
  twistMotion=Free
 }
]
