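# Post-copy AnimBP for the female Witch body, adding secondary motion to dress, front flap and accessory bones.
# Copied pose feeds RigidBody, thigh constraints and AnimDynamics chains; crouching reduces side-panel simulation.
# Knee bend lifts back panels, and an open parachute adds a skirt_open additive.
# EventGraph tunes cloth drives for skydiving and computes altitude-noise jiggle and force, but their AnimGraph nodes, like the crouch PoseDriver, are disconnected from output.
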AnimBlueprint /Game/Characters/Player/Female/Medium/Bodies/F_MED_Witch/Meshes/F_MED_Witch_ABP
target_skeleton=&/F_MED_Witch_Skeleton
graphs [
 AnimGraph {nodes [
   Look At - Bone: dyn_flap_front_01(Alpha=1) {
    alphaBoolBlend {blendOption=Linear}
    boneToModify {boneName=dyn_flap_front_01}
    lookAtLocation=<-2 -2>
    lookAtTarget {boneReference {boneName=dyn_dress_front_l_01}}
    lookAt_Axis {axis {y=0, z=-1}}
   }
   n1 = Get knee_right() -> (knee_right:double)
   "Transform (Modify) Bone - Bone: dyn_dress_back_r_03"(Alpha=@n1, Translation="0,0,0", Rotation="0.000,0.000,15", Scale="1,1,1") {
    rotationMode=BMM_Additive
    alphaBoolBlend {blendOption=Linear}
    alphaScaleBiasClamp {bInterpResult=true}
    boneToModify {boneName=dyn_dress_back_r_03}
    rotation {roll=15}
   }
   n2 = Get knee_left() -> (knee_left:double)
   "Transform (Modify) Bone - Bone: dyn_dress_back_l_03"(Alpha=@n2, Translation="0,0,0", Rotation="0.000,0.000,15", Scale="1,1,1") {
    rotationMode=BMM_Additive
    alphaBoolBlend {blendOption=Linear}
    alphaScaleBiasClamp {bInterpResult=true}
    boneToModify {boneName=dyn_dress_back_l_03}
    rotation {roll=15}
   }
   n3 = Get current_dive_jiggle() -> (current_dive_jiggle:Rotator)
   n4 = "Transform (Modify) Bone - Bone: dyn_dress_back_l_01"(Alpha=0.7, Translation="0,0,0", Rotation=@n3, Scale="1,1,1") {rotationMode=BMM_Additive, rotationSpace=BCS_BoneSpace, alphaBoolBlend {blendOption=Linear}, boneToModify {boneName=dyn_dress_back_l_01}}
   n5 = "Transform (Modify) Bone - Bone: dyn_dress_back_r_01"(ComponentPose=@n4, Alpha=0.7, Translation="0,0,0", Rotation=@n3, Scale="1,1,1") {rotationMode=BMM_Additive, rotationSpace=BCS_BoneSpace, alphaBoolBlend {blendOption=Linear}, boneToModify {boneName=dyn_dress_back_r_01}}
   n6 = "Transform (Modify) Bone - Bone: dyn_dress_side_l_01"(ComponentPose=@n5, Alpha=0.7, Translation="0,0,0", Rotation=@n3, Scale="1,1,1") {rotationMode=BMM_Additive, rotationSpace=BCS_BoneSpace, alphaBoolBlend {blendOption=Linear}, boneToModify {boneName=dyn_dress_side_l_01}}
   n7 = "Transform (Modify) Bone - Bone: dyn_dress_side_r_01"(ComponentPose=@n6, Alpha=0.7, Translation="0,0,0", Rotation=@n3, Scale="1,1,1") {rotationMode=BMM_Additive, rotationSpace=BCS_BoneSpace, alphaBoolBlend {blendOption=Linear}, boneToModify {boneName=dyn_dress_side_r_01}}
   n8 = "Transform (Modify) Bone - Bone: dyn_dress_back_l_01"(ComponentPose=@n7, Alpha=0.7, Translation="0,0,0", Rotation=@n3, Scale="1,1,1") {rotationMode=BMM_Additive, rotationSpace=BCS_BoneSpace, alphaBoolBlend {blendOption=Linear}, boneToModify {boneName=dyn_dress_back_l_01}}
   n9 = "Transform (Modify) Bone - Bone: dyn_dress_back_r_01"(ComponentPose=@n8, Alpha=0.7, Translation="0,0,0", Rotation=@n3, Scale="1,1,1") {rotationMode=BMM_Additive, rotationSpace=BCS_BoneSpace, alphaBoolBlend {blendOption=Linear}, boneToModify {boneName=dyn_dress_back_r_01}}
   n10 = "Transform (Modify) Bone - Bone: dyn_flap_back"(ComponentPose=@n9, Alpha=1, Translation="0,0,0", Rotation=@n3, Scale="1,1,1") {rotationMode=BMM_Additive, rotationSpace=BCS_BoneSpace, alphaBoolBlend {blendOption=Linear}, boneToModify {boneName=dyn_flap_back}}
   n11 = "Transform (Modify) Bone - Bone: dyn_book"(ComponentPose=@n10, Alpha=1, Translation="0,0,0", Rotation=@n3, Scale="1,1,1") {rotationMode=BMM_Additive, rotationSpace=BCS_BoneSpace, alphaBoolBlend {blendOption=Linear}, boneToModify {boneName=dyn_book}}
   n12 = "Transform (Modify) Bone - Bone: dyn_potion"(ComponentPose=@n11, Alpha=1, Translation="0,0,0", Rotation=@n3, Scale="1,1,1") {rotationMode=BMM_Additive, rotationSpace=BCS_BoneSpace, alphaBoolBlend {blendOption=Linear}, boneToModify {boneName=dyn_potion}}
   "Transform (Modify) Bone - Bone: dyn_waist_sack"(ComponentPose=@n12, Alpha=1, Translation="0,0,0", Rotation=@n3, Scale="1,1,1") {rotationMode=BMM_Additive, rotationSpace=BCS_BoneSpace, alphaBoolBlend {blendOption=Linear}, boneToModify {boneName=dyn_waist_sack}}
   Pose Driver - Source: calf_l!(SourcePose:Pose Link Structure) -> (Pose:Pose Link Structure) {
    is=title AnimGraphNode_PoseDriver
    poseAsset=/Game/Characters/Player/Female/Medium/Bodies/F_MED_Witch/Meshes/poses/F_MED_Witch_Crouch_PoseAsset
    onlyDriveBones [
     {boneName=dyn_dress_back_r_01}
     {boneName=dyn_dress_back_r_02}
     {boneName=dyn_dress_back_r_03}
     {boneName=dyn_dress_back_l_01}
     {boneName=dyn_dress_front_l_02}
     {boneName=dyn_dress_front_l_03}
    ]
    poseTargets [{
      bApplyCustomCurve=false
      bIsHidden=false
      distanceMethod=DefaultMethod
      drivenName=crouching_pose
      functionType=DefaultFunction
      targetScale=1
      boneTransforms [
       {targetRotation=<rot 0.2025 112.4855 0.6091>, targetTranslation=<-42.2357 0.343 -0.726>}
       {targetRotation=<rot 0.1561 117.0401 0.4504>, targetTranslation=<42.3115 -0.3586 0.723>}
       {targetRotation=<rot -2.019 -10.6495 -9.005>, targetTranslation=<-46.7521 -2.502 -0.5919>}
       {targetRotation=<rot 9.3697 28.3207 -55.875>, targetTranslation=<46.6874 2.4987 0.5955>}
       {targetRotation=<rot 17.1818 147.3521 -11.6723>, targetTranslation=<-1.0853 -0.2108 8.7051>}
      ]
      customCurve {defaultValue=3.4028234663852886e+38, postInfinityExtrap=RCCE_Constant, preInfinityExtrap=RCCE_Constant, keys []}
      targetRotation=<rot 0 0 0>
     }]
    rBFParams {radius=20}
    sourceBones [{boneName=calf_l}, {boneName=calf_r}, {boneName=foot_l}, {boneName=foot_r}, {boneName=thigh_r}]
   }
   RigidBody() {
    bOverrideWorldGravity=true
    overridePhysicsAsset=&/F_MED_Witch_Skirt_Tweak
    simulationSpace=BaseBoneSpace
    alphaBoolBlend {blendOption=Linear}
    baseBoneRef {boneName=root}
    overrideWorldGravity {z=-980}
   }
   RigidBody() {
    bOverrideWorldGravity=true
    overridePhysicsAsset=&/F_MED_Witch_Skirt_Phys
    simulationSpace=BaseBoneSpace
    alphaBoolBlend {blendOption=Linear}
    baseBoneRef {boneName=root}
    componentAppliedLinearAccClamp=<1000 1000 1000>
    componentLinearAccScale=<0.1 0.1 0.1>
    componentLinearVelScale=<0.1 0.1 0.1>
    overrideWorldGravity {z=-980}
   }
   RigidBody() {
    bOverrideWorldGravity=true
    overridePhysicsAsset=&/F_MED_Witch_Loincloth_Phys
    simulationSpace=BaseBoneSpace
    alphaBoolBlend {blendOption=Linear}
    baseBoneRef {boneName=root}
    overrideWorldGravity {z=-980}
   }
   n13 = RigidBody(ComponentPose=Local To Component(LocalPose=Copy Pose From Mesh() {bUseAttachedParent=true})) {
    bOverrideWorldGravity=true
    overridePhysicsAsset=&/F_MED_Witch_Physics
    simulationSpace=BaseBoneSpace
    alphaBoolBlend {blendOption=Linear}
    baseBoneRef {boneName=root}
    overrideWorldGravity {z=-980}
   }
   n14 = "Constraint - dyn_dress_front_r_01 (P)"(ComponentPose=@n13, Alpha=1, ConstraintWeights_0=1) {
    alphaBoolBlend {blendOption=Linear}
    boneToModify {boneName=dyn_dress_front_r_01}
    constraintSetup [{targetBone {boneName=thigh_r}}]
    constraintWeights [1]
   }
   n15 = "Constraint - dyn_dress_front_l_01 (P)"(ComponentPose=@n14, Alpha=1, ConstraintWeights_0=1) {
    alphaBoolBlend {blendOption=Linear}
    boneToModify {boneName=dyn_dress_front_l_01}
    constraintSetup [{targetBone {boneName=thigh_l}}]
    constraintWeights [1]
   }
   n16 = Anim Dynamics - Bone: dyn_dress_back_r_01(ComponentPose=@n15, Alpha=1, ExternalForce="0,0,0") {
    angularDampingOverride=0.9999
    bGravityOverrideInSimSpace=true
    bOverrideAngularDamping=true
    gravityScale=0.35
    simulationSpace=RootRelative
    alphaBoolBlend {blendOption=Linear}
    boundBone {boneName=dyn_dress_back_r_01}
    componentAppliedLinearAccClamp=<100 100 100>
    physicsBodyDefinitions [{
      sphereCollisionRadius=0
      boundBone {boneName=dyn_dress_back_r_01}
      boxExtents=<5 20 20>
      constraintSetup {angularLimitsMax=<0 0 30>, angularLimitsMin=<0 0 0>, angularTarget=<0 0 0>}
      localJointOffset=<5 0 0>
     }]
   }
   n17 = Anim Dynamics - Bone: dyn_dress_back_r_02(ComponentPose=@n16, Alpha=1, ExternalForce="0,0,0") {
    angularDampingOverride=0.9999
    bGravityOverrideInSimSpace=true
    bOverrideAngularDamping=true
    gravityScale=0.35
    simulationSpace=RootRelative
    alphaBoolBlend {blendOption=Linear}
    boundBone {boneName=dyn_dress_back_r_02}
    componentAppliedLinearAccClamp=<100 100 100>
    physicsBodyDefinitions [{
      sphereCollisionRadius=0
      boundBone {boneName=dyn_dress_back_r_02}
      boxExtents=<5 20 20>
      constraintSetup {angularLimitsMax=<0 0 90>, angularLimitsMin=<0 0 0>, angularTarget=<0 0 0>}
      localJointOffset=<5 0 0>
     }]
   }
   n18 = Anim Dynamics - Bone: dyn_dress_back_r_03(ComponentPose=@n17, Alpha=1, ExternalForce="0,0,0") {
    angularDampingOverride=0.9999
    bGravityOverrideInSimSpace=true
    bOverrideAngularDamping=true
    gravityScale=0.35
    simulationSpace=RootRelative
    alphaBoolBlend {blendOption=Linear}
    boundBone {boneName=dyn_dress_back_r_03}
    componentAppliedLinearAccClamp=<100 100 100>
    physicsBodyDefinitions [{
      sphereCollisionRadius=0
      boundBone {boneName=dyn_dress_back_r_03}
      boxExtents=<5 20 20>
      constraintSetup {angularLimitsMax=<0 0 90>, angularLimitsMin=<0 0 0>, angularTarget=<0 0 0>}
      localJointOffset=<5 0 0>
     }]
   }
   n19 = Get crouch_alpha() -> (crouch_alpha:double)
   n20 = Anim Dynamics - Bone: dyn_dress_side_r_01(ComponentPose=@n18, Alpha=@n19, ExternalForce="0,0,0") {
    angularDampingOverride=0.9999
    bGravityOverrideInSimSpace=true
    bOverrideAngularDamping=true
    gravityScale=0.35
    simulationSpace=RootRelative
    alphaBoolBlend {blendOption=Linear}
    boundBone {boneName=dyn_dress_side_r_01}
    componentAppliedLinearAccClamp=<100 100 100>
    physicsBodyDefinitions [{
      sphereCollisionRadius=0
      boundBone {boneName=dyn_dress_side_r_01}
      boxExtents=<5 20 20>
      constraintSetup {angularLimitsMax=<0 20 0>, angularLimitsMin=<0 5 0>, angularTarget=<0 0 0>}
      localJointOffset=<5 0 0>
     }]
   }
   n21 = Anim Dynamics - Bone: dyn_dress_side_r_02(ComponentPose=@n20, Alpha=@n19, ExternalForce="0,0,0") {
    angularDampingOverride=0.9999
    bGravityOverrideInSimSpace=true
    bOverrideAngularDamping=true
    gravityScale=0.35
    simulationSpace=RootRelative
    alphaBoolBlend {blendOption=Linear}
    boundBone {boneName=dyn_dress_side_r_02}
    componentAppliedLinearAccClamp=<100 100 100>
    physicsBodyDefinitions [{
      sphereCollisionRadius=0
      boundBone {boneName=dyn_dress_side_r_02}
      boxExtents=<5 20 20>
      constraintSetup {angularLimitsMax=<0 20 0>, angularLimitsMin=<0 0 0>, angularTarget=<0 0 0>}
      localJointOffset=<5 0 0>
     }]
   }
   n22 = Anim Dynamics - Bone: dyn_dress_side_r_03(ComponentPose=@n21, Alpha=@n19, ExternalForce="0,0,0") {
    angularDampingOverride=0.9999
    bGravityOverrideInSimSpace=true
    bOverrideAngularDamping=true
    bUsePlanarLimit=false
    gravityScale=0.35
    simulationSpace=RootRelative
    alphaBoolBlend {blendOption=Linear}
    boundBone {boneName=dyn_dress_side_r_03}
    componentAppliedLinearAccClamp=<100 100 100>
    physicsBodyDefinitions [{
      collisionType=CustomSphere
      sphereCollisionRadius=1
      boundBone {boneName=dyn_dress_side_r_03}
      boxExtents=<5 20 20>
      constraintSetup {angularLimitsMax=<0 30 0>, angularLimitsMin=<0 0 0>, angularTarget=<0 0 0>}
      localJointOffset=<5 0 0>
     }]
   }
   n23 = Anim Dynamics - Bone: dyn_dress_front_r_01(ComponentPose=@n22, Alpha=1, ExternalForce="0,0,0") {
    angularDampingOverride=0.9999
    bGravityOverrideInSimSpace=true
    bOverrideAngularDamping=true
    gravityScale=0.35
    simulationSpace=RootRelative
    alphaBoolBlend {blendOption=Linear}
    boundBone {boneName=dyn_dress_front_r_01}
    componentAppliedLinearAccClamp=<100 100 100>
    physicsBodyDefinitions [{
      sphereCollisionRadius=0
      boundBone {boneName=dyn_dress_front_r_01}
      boxExtents=<5 20 20>
      constraintSetup {angularLimitsMax=<0 0 0>, angularLimitsMin=<0 0 -40>, angularTarget=<0 0 0>}
      localJointOffset=<5 0 0>
     }]
   }
   n24 = Anim Dynamics - Bone: dyn_dress_front_r_02(ComponentPose=@n23, Alpha=1, ExternalForce="0,0,0") {
    angularDampingOverride=0.99
    bGravityOverrideInSimSpace=true
    bOverrideAngularDamping=true
    gravityScale=0.35
    simulationSpace=RootRelative
    alphaBoolBlend {blendOption=Linear}
    boundBone {boneName=dyn_dress_front_r_02}
    componentAppliedLinearAccClamp=<100 100 100>
    componentLinearAccScale=<1 1 1>
    componentLinearVelScale=<1 1 1>
    physicsBodyDefinitions [{
      sphereCollisionRadius=0
      boundBone {boneName=dyn_dress_front_r_02}
      boxExtents=<5 20 20>
      constraintSetup {angularLimitsMax=<0 0 2>, angularLimitsMin=<0 0 -20>, angularTarget=<0 0 0>}
      localJointOffset=<5 0 0>
     }]
   }
   n25 = Anim Dynamics - Bone: dyn_dress_front_r_03(ComponentPose=@n24, Alpha=1, ExternalForce="0,0,0") {
    angularDampingOverride=0.99
    bGravityOverrideInSimSpace=true
    bOverrideAngularDamping=true
    bUsePlanarLimit=false
    gravityScale=0.35
    simulationSpace=RootRelative
    alphaBoolBlend {blendOption=Linear}
    boundBone {boneName=dyn_dress_front_r_03}
    componentAppliedLinearAccClamp=<100 100 100>
    componentLinearAccScale=<1 1 1>
    componentLinearVelScale=<1 1 1>
    physicsBodyDefinitions [{
      collisionType=CustomSphere
      sphereCollisionRadius=1
      boundBone {boneName=dyn_dress_front_r_03}
      boxExtents=<5 20 20>
      constraintSetup {angularLimitsMax=<0 0 5>, angularLimitsMin=<0 0 -40>, angularTarget=<0 0 0>}
      localJointOffset=<5 0 0>
     }]
   }
   n26 = Anim Dynamics - Bone: dyn_dress_back_l_01(ComponentPose=@n25, Alpha=1, ExternalForce="0,0,0") {
    angularDampingOverride=0.9999
    bGravityOverrideInSimSpace=true
    bOverrideAngularDamping=true
    gravityScale=0.35
    simulationSpace=RootRelative
    alphaBoolBlend {blendOption=Linear}
    boundBone {boneName=dyn_dress_back_l_01}
    componentAppliedLinearAccClamp=<100 100 100>
    physicsBodyDefinitions [{
      sphereCollisionRadius=0
      boundBone {boneName=dyn_dress_back_l_01}
      boxExtents=<5 20 20>
      constraintSetup {angularLimitsMax=<0 0 30>, angularLimitsMin=<0 0 0>, angularTarget=<0 0 0>}
      localJointOffset=<5 0 0>
     }]
   }
   n27 = Anim Dynamics - Bone: dyn_dress_back_l_02(ComponentPose=@n26, Alpha=1, ExternalForce="0,0,0") {
    angularDampingOverride=0.9999
    bGravityOverrideInSimSpace=true
    bOverrideAngularDamping=true
    gravityScale=0.35
    simulationSpace=RootRelative
    alphaBoolBlend {blendOption=Linear}
    boundBone {boneName=dyn_dress_back_l_02}
    componentAppliedLinearAccClamp=<100 100 100>
    physicsBodyDefinitions [{
      sphereCollisionRadius=0
      boundBone {boneName=dyn_dress_back_l_02}
      boxExtents=<5 20 20>
      constraintSetup {angularLimitsMax=<0 0 40>, angularLimitsMin=<0 0 0>, angularTarget=<0 0 0>}
      localJointOffset=<5 0 0>
     }]
   }
   n28 = Anim Dynamics - Bone: dyn_dress_back_l_03(ComponentPose=@n27, Alpha=1, ExternalForce="0,0,0") {
    angularDampingOverride=0.9999
    bGravityOverrideInSimSpace=true
    bOverrideAngularDamping=true
    gravityScale=0.35
    simulationSpace=RootRelative
    alphaBoolBlend {blendOption=Linear}
    boundBone {boneName=dyn_dress_back_l_03}
    componentAppliedLinearAccClamp=<100 100 100>
    physicsBodyDefinitions [{
      sphereCollisionRadius=0
      boundBone {boneName=dyn_dress_back_l_03}
      boxExtents=<5 20 20>
      constraintSetup {angularLimitsMax=<0 0 30>, angularLimitsMin=<0 0 0>, angularTarget=<0 0 0>}
      localJointOffset=<5 0 0>
     }]
   }
   n29 = Get crouch_alpha()
   n30 = Anim Dynamics - Bone: dyn_dress_side_l_01(ComponentPose=@n28, Alpha=@n29, ExternalForce="0,0,0") {
    angularDampingOverride=0.9999
    bGravityOverrideInSimSpace=true
    bOverrideAngularDamping=true
    gravityScale=0.35
    simulationSpace=RootRelative
    alphaBoolBlend {blendOption=Linear}
    boundBone {boneName=dyn_dress_side_l_01}
    componentAppliedLinearAccClamp=<100 100 100>
    physicsBodyDefinitions [{
      sphereCollisionRadius=0
      boundBone {boneName=dyn_dress_side_l_01}
      boxExtents=<5 20 20>
      constraintSetup {angularLimitsMax=<0 -5 0>, angularLimitsMin=<0 -20 0>, angularTarget=<0 0 0>}
      localJointOffset=<5 0 0>
     }]
   }
   n31 = Anim Dynamics - Bone: dyn_dress_side_l_02(ComponentPose=@n30, Alpha=@n29, ExternalForce="0,0,0") {
    angularDampingOverride=0.9999
    bGravityOverrideInSimSpace=true
    bOverrideAngularDamping=true
    gravityScale=0.35
    simulationSpace=RootRelative
    alphaBoolBlend {blendOption=Linear}
    boundBone {boneName=dyn_dress_side_l_02}
    componentAppliedLinearAccClamp=<100 100 100>
    physicsBodyDefinitions [{
      sphereCollisionRadius=0
      boundBone {boneName=dyn_dress_side_l_02}
      boxExtents=<5 20 20>
      constraintSetup {angularLimitsMax=<0 0 0>, angularLimitsMin=<0 -20 0>, angularTarget=<0 0 0>}
      localJointOffset=<5 0 0>
     }]
   }
   n32 = Anim Dynamics - Bone: dyn_dress_side_l_03(ComponentPose=@n31, Alpha=@n29, ExternalForce="0,0,0") {
    angularDampingOverride=0.9999
    bGravityOverrideInSimSpace=true
    bOverrideAngularDamping=true
    bUsePlanarLimit=false
    gravityScale=0.35
    simulationSpace=RootRelative
    alphaBoolBlend {blendOption=Linear}
    boundBone {boneName=dyn_dress_side_l_03}
    componentAppliedLinearAccClamp=<100 100 100>
    physicsBodyDefinitions [{
      collisionType=CustomSphere
      sphereCollisionRadius=1
      boundBone {boneName=dyn_dress_side_l_03}
      boxExtents=<5 20 20>
      constraintSetup {angularLimitsMax=<0 0 0>, angularLimitsMin=<0 -30 0>, angularTarget=<0 0 0>}
      localJointOffset=<5 0 0>
     }]
   }
   n33 = Anim Dynamics - Bone: dyn_dress_front_l_01(ComponentPose=@n32, Alpha=1, ExternalForce="0,0,0") {
    angularDampingOverride=0.9999
    bGravityOverrideInSimSpace=true
    bOverrideAngularDamping=true
    gravityScale=0.35
    simulationSpace=RootRelative
    alphaBoolBlend {blendOption=Linear}
    boundBone {boneName=dyn_dress_front_l_01}
    componentAppliedLinearAccClamp=<100 100 100>
    physicsBodyDefinitions [{
      sphereCollisionRadius=0
      boundBone {boneName=dyn_dress_front_l_01}
      boxExtents=<5 20 20>
      constraintSetup {angularLimitsMax=<0 0 0>, angularLimitsMin=<0 0 -40>, angularTarget=<0 0 0>}
      localJointOffset=<5 0 0>
     }]
   }
   n34 = Anim Dynamics - Bone: dyn_dress_front_l_02(ComponentPose=@n33, Alpha=1, ExternalForce="0,0,0") {
    angularDampingOverride=0.99
    bGravityOverrideInSimSpace=true
    bOverrideAngularDamping=true
    gravityScale=0.35
    simulationSpace=RootRelative
    alphaBoolBlend {blendOption=Linear}
    boundBone {boneName=dyn_dress_front_l_02}
    componentAppliedLinearAccClamp=<100 100 100>
    componentLinearAccScale=<1 1 1>
    componentLinearVelScale=<1 1 1>
    physicsBodyDefinitions [{
      sphereCollisionRadius=0
      boundBone {boneName=dyn_dress_front_l_02}
      boxExtents=<5 20 20>
      constraintSetup {angularLimitsMax=<0 0 2>, angularLimitsMin=<0 0 -20>, angularTarget=<0 0 0>}
      localJointOffset=<5 0 0>
     }]
   }
   n35 = Anim Dynamics - Bone: dyn_dress_front_l_03(ComponentPose=@n34, Alpha=1, ExternalForce="0,0,0") {
    angularDampingOverride=0.99
    bGravityOverrideInSimSpace=true
    bOverrideAngularDamping=true
    bUsePlanarLimit=false
    gravityScale=0.35
    simulationSpace=RootRelative
    alphaBoolBlend {blendOption=Linear}
    boundBone {boneName=dyn_dress_front_l_03}
    componentAppliedLinearAccClamp=<100 100 100>
    componentLinearAccScale=<1 1 1>
    componentLinearVelScale=<1 1 1>
    physicsBodyDefinitions [{
      collisionType=CustomSphere
      sphereCollisionRadius=1
      boundBone {boneName=dyn_dress_front_l_03}
      boxExtents=<5 20 20>
      constraintSetup {angularLimitsMax=<0 0 5>, angularLimitsMin=<0 0 -40>, angularTarget=<0 0 0>}
      localJointOffset=<5 0 0>
     }]
   }
   n36 = "Constraint - dyn_flap_front_01 (PPP)"!(ComponentPose=@n35, Alpha=1, ConstraintWeights_0=1, ConstraintWeights_1:float=1, ConstraintWeights_2:float=1) -> (Pose) {
    alphaBoolBlend {blendOption=Linear}
    boneToModify {boneName=dyn_flap_front_01}
    constraintSetup [{targetBone {boneName=thigh_l}}, {targetBone {boneName=thigh_r}}, {targetBone {boneName=pelvis}}]
    constraintWeights [1,1,1]
   }
   n37 = "Transform (Modify) Bone - Bone: dyn_flap_front_01"(ComponentPose=@n36, Alpha=1, Translation="0,0,0", Rotation="0.413,8.433,-5.598", Scale="1,1,1") {
    rotationMode=BMM_Additive
    alphaBoolBlend {blendOption=Linear}
    boneToModify {boneName=dyn_flap_front_01}
    rotation=<rot 0.413 8.4333 -5.5981>
   }
   n38 = Anim Dynamics - Bone: dyn_flap_front_01(ComponentPose=@n37, Alpha=1, ExternalForce="0,0,0") {
    angularDampingOverride=0.9999
    bGravityOverrideInSimSpace=true
    bOverrideAngularDamping=true
    gravityScale=0.35
    simulationSpace=RootRelative
    alphaBoolBlend {blendOption=Linear}
    boundBone {boneName=dyn_flap_front_01}
    componentAppliedLinearAccClamp=<100 100 100>
    componentLinearAccScale=<1 1 1>
    componentLinearVelScale=<1 1 1>
    physicsBodyDefinitions [{
      sphereCollisionRadius=1
      boundBone {boneName=dyn_flap_front_01}
      boxExtents=<5 20 20>
      constraintSetup {angularLimitsMax=<0 0 -2>, angularLimitsMin=<0 0 -40>, angularTarget=<0 0 0>}
      localJointOffset=<5 0 0>
     }]
   }
   n39 = Anim Dynamics - Bone: dyn_flap_front_02(ComponentPose=@n38, Alpha=1, ExternalForce="0,0,0") {
    angularDampingOverride=0.9999
    bGravityOverrideInSimSpace=true
    bOverrideAngularDamping=true
    gravityScale=0.35
    simulationSpace=RootRelative
    alphaBoolBlend {blendOption=Linear}
    boundBone {boneName=dyn_flap_front_02}
    componentAppliedLinearAccClamp=<100 100 100>
    componentLinearAccScale=<1 1 1>
    componentLinearVelScale=<1 1 1>
    physicsBodyDefinitions [{
      sphereCollisionRadius=1
      boundBone {boneName=dyn_flap_front_02}
      boxExtents=<5 20 20>
      constraintSetup {angularLimitsMax=<0 0 5>, angularLimitsMin=<0 0 -40>, angularTarget=<0 0 0>}
      localJointOffset=<5 0 0>
     }]
   }
   n40 = Anim Dynamics - Bone: dyn_flap_front_03(ComponentPose=@n39, Alpha=1, ExternalForce="0,0,0") {
    angularDampingOverride=0.9999
    bGravityOverrideInSimSpace=true
    bOverrideAngularDamping=true
    gravityScale=0.35
    simulationSpace=RootRelative
    alphaBoolBlend {blendOption=Linear}
    boundBone {boneName=dyn_flap_front_03}
    componentAppliedLinearAccClamp=<100 100 100>
    componentLinearAccScale=<1 1 1>
    componentLinearVelScale=<1 1 1>
    physicsBodyDefinitions [{
      sphereCollisionRadius=1
      boundBone {boneName=dyn_flap_front_03}
      boxExtents=<5 20 20>
      constraintSetup {angularLimitsMax=<0 10 10>, angularLimitsMin=<0 -10 -40>, angularTarget=<0 0 0>}
      localJointOffset=<5 0 0>
     }]
   }
   n41 = "Transform (Modify) Bone - Bone: dyn_dress_back_l_01"(ComponentPose=@n40, Alpha=@n2, Translation="0,0,0", Rotation="0.000,0.000,35", Scale="1,1,1") {
    rotationMode=BMM_Additive
    alphaBoolBlend {blendOption=Linear}
    alphaScaleBiasClamp {bInterpResult=true}
    boneToModify {boneName=dyn_dress_back_l_01}
    rotation {roll=35}
   }
   n42 = "Transform (Modify) Bone - Bone: dyn_dress_back_l_02"(ComponentPose=@n41, Alpha=@n2, Translation="0,0,0", Rotation="0.000,0.000,20", Scale="1,1,1") {
    rotationMode=BMM_Additive
    alphaBoolBlend {blendOption=Linear}
    alphaScaleBiasClamp {bInterpResult=true}
    boneToModify {boneName=dyn_dress_back_l_02}
    rotation {roll=20}
   }
   n43 = "Transform (Modify) Bone - Bone: dyn_dress_back_r_01"(ComponentPose=@n42, Alpha=@n1, Translation="0,0,0", Rotation="0.000,0.000,35", Scale="1,1,1") {
    rotationMode=BMM_Additive
    alphaBoolBlend {blendOption=Linear}
    alphaScaleBiasClamp {bInterpResult=true}
    boneToModify {boneName=dyn_dress_back_r_01}
    rotation {roll=35}
   }
   n44 = "Transform (Modify) Bone - Bone: dyn_dress_back_r_02"(ComponentPose=@n43, Alpha=@n1, Translation="0,0,0", Rotation="0.000,0.000,20", Scale="1,1,1") {
    rotationMode=BMM_Additive
    alphaBoolBlend {blendOption=Linear}
    alphaScaleBiasClamp {bInterpResult=true}
    boneToModify {boneName=dyn_dress_back_r_02}
    rotation {roll=20}
   }
   Output Pose(Result=Apply Additive!(
     Base=Component To Local(ComponentPose=@n44)
     Additive=F_MED_Witch_Skirt_PoseAsset : skirt_open!(PoseWeight:float=0.4) -> (Pose:Pose Link Structure) {
      is=title AnimGraphNode_PoseByName
      poseAsset=/Game/Characters/Player/Female/Medium/Bodies/F_MED_Witch/Meshes/poses/F_MED_Witch_Skirt_PoseAsset
      poseName=skirt_open
     }
     bAlphaBoolEnabled:bool=Get bIsParachuteOpened()
    ) -> (Pose) {alphaInputType=Bool, alphaBoolBlend {blendInTime=2, blendOption=Linear}}) {layerGroup=None}
   Break Fort Anim Input Skydiving External Force(FortAnimInput_SkydivingExternalForce=Get SkydivingExternalForceInput()) -> (
    bUseSkydivingVectorForce=~
    SkydivingVectorForce=~
    HeadToPelvisDirection=~
    FloatingMultiplier=~
    FloatingAdditive=~
    DivingMultiplier=~
    DivingAdditive=~
    ParachutingMultiplier=~
    ParachutingAdditive=~
   )
   Get falling_anim_dyn_up_force_vec3() -> (falling_anim_dyn_up_force_vec3:Vector)
  ]}
 EventGraph {nodes [
   n1 = Event BlueprintUpdateAnimation()
   > n2 = Sequence!(execute) -> (then_0, then_1, then_2:Exec)
   n3 = BreakVector(InVec=Get PawnLastLocation() -> (PawnLastLocation:Vector))
   > Set current_actor_z(current_actor_z:double=@n3.Z) -> (Output_Get:double)
   n4 = SelectFloat(A=0.25, B=1, bPickA=Get bIsParachuteOpened())
   n5 = float + float(A=@n3.Z, B=20) {is=title K2Node_CommutativeAssociativeBinaryOperator}
   n6 = Get bIsSkyDiving()
   n7 = MakeRotator(
    Pitch=SelectFloat(A=Make Perlin Noise and Remap!(Value:float=@n3.Z, RangeOutMin:float=-30, RangeOutMax:float=40) -> (ReturnValue:float), B=0, bPickA=@n6)
    Yaw=SelectFloat(A=Make Perlin Noise and Remap(Value=@n5, RangeOutMin=-30, RangeOutMax=30), B=0, bPickA=@n6)
   )
   > Set current_dive_jiggle(current_dive_jiggle:Rotator=RInterpTo!(
     Current:Rotator=Get current_dive_jiggle()
     Target:Rotator="Scale Rotator (Float)"!(A:Rotator=@n7, B:float=@n4) -> (ReturnValue:Rotator)
     DeltaTime:float=@n1.DeltaTimeX
     InterpSpeed:float=10
    ) -> (ReturnValue:Rotator)) -> (Output_Get:Rotator)
   n8 = SelectFloat(
    A=Make Perlin Noise and Remap(Value=Get current_actor_z() -> (current_actor_z:double), RangeOutMin=0, RangeOutMax=5000)
    B=0
    bPickA=Get bIsSkyDiving()
   )
   n9 = float * float(A=@n8, B=0.5)
   n10 = Get bIsParachuteOpened()
   > Set falling_anim_dyn_up_force_vec3(falling_anim_dyn_up_force_vec3:Vector=MakeVector(Y=0, Z=SelectFloat(A=@n9, B=@n8, bPickA=@n10))) -> (Output_Get:Vector)
   > n11 = Branch(Condition=Get bIsSkyDiving())
   n12 = Get clothing_sim_interactor() -> (clothing_sim_interactor:Clothing Simulation Interactor Nv Object Reference)
   > SetAnimDriveSpringStiffness!(execute, self:Clothing Simulation Interactor Object Reference=@n12, InStiffness:float=10) -> (then)
   > SetAnimDriveDamperStiffness!(execute, self:Clothing Simulation Interactor Nv Object Reference=@n12, InStiffness:float=5) -> (then)
   > Do Once(Reset=@n13.then, Start Closed=true)
   > ForceClothNextUpdateTeleportAndReset(self:Skeletal Mesh Component Object Reference=Get MeshToCopy() -> (MeshToCopy:Skeletal Mesh Component Object Reference))
   SetAnimDriveSpringStiffness(execute=@n11.else, self=@n12, InStiffness=0)
   > n13 = SetAnimDriveDamperStiffness(self=@n12, InStiffness=0)
   n14 = GetDeltaTransformFromRefPose(self=GetPartSkeletalMeshComponent(), BoneName=thigh_r, BaseName=calf_r)
   n15 = MapRangeClamped(Value=BreakRotator(InRot=BreakTransform(InTransform=@n14).Rotation).Yaw, InRangeA=-80, InRangeB=-120)
   Set knee_right(execute=@n2.then_2, knee_right:double=@n15) -> (Output_Get:double)
   n16 = GetDeltaTransformFromRefPose(self=GetPartSkeletalMeshComponent(), BoneName=thigh_l, BaseName=calf_l)
   n17 = MapRangeClamped(Value=BreakRotator(InRot=BreakTransform(InTransform=@n16).Rotation).Yaw, InRangeA=-80, InRangeB=-120)
   Set knee_left(execute=@n2.then_1, knee_left:double=@n17) -> (Output_Get:double)
   Event OnBeginCrouching()
   > Set crouch_alpha(crouch_alpha:double=0.25) -> (Output_Get:double)
   Event OnEndCrouching()
   > Set crouch_alpha(crouch_alpha=1)
   Event BlueprintInitializeAnimation()
   n18 = GetPartSkeletalMeshComponent()
   > n19 = Cast To ClothingSimulationInteractorNv!(
    execute
    Object=GetClothingSimulationInteractor!(self:Skeletal Mesh Component Object Reference=@n18, ClothingAssetName:Name=None) -> (ReturnValue:Clothing Simulation Interactor Object Reference)
   ) -> (then, CastFailed, AsClothing Simulation Interactor Nv:Clothing Simulation Interactor Nv Object Reference) {is=title K2Node_DynamicCast}
   > Set clothing_sim_interactor(clothing_sim_interactor:Clothing Simulation Interactor Nv Object Reference=@n19) -> (Output_Get:Clothing Simulation Interactor Nv Object Reference)
   > Set ClothBlendWeight!(execute, ClothBlendWeight:float=0, self:Skeletal Mesh Component Object Reference=@n18) -> (then, Output_Get:float)
   SelectFloat(A=@n17, B=0, bPickA=Get bIsCrouching())
   SelectFloat(A=@n15, B=0, bPickA=Get bIsCrouching())
  ]}
]
variables [
 clothing_sim_interactor {type=ClothingSimulationInteractorNv, default=null}
 crouch_alpha {type=number}
 current_actor_z {type=number}
 current_dive_jiggle {type=Rotator}
 falling_anim_dyn_up_force_vec3 {type=Vector}
 knee_left {type=number}
 knee_right {type=number}
 MeshToCopy {type=SkeletalMeshComponent}
]
