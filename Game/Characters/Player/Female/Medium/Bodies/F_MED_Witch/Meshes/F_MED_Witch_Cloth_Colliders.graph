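# Physics asset for the female medium Witch player body, named as the cloth collider set for that character.
# It has bodies on the pelvis, thighs and calves, but only the pelvis lists a collision shape, a single capsule.
# Constraints chain each calf to its thigh and each thigh to the pelvis, with swing and twist left free.
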PhysicsAsset /Game/Characters/Player/Female/Medium/Bodies/F_MED_Witch/Meshes/F_MED_Witch_Cloth_Colliders
bodies {calf_l [], calf_r [], pelvis [{shapeType=Capsule, length=11.4212, radius=7.4885, center=<-5.8999 0.7238 -0.0002>}], thigh_l [], thigh_r []}
constraints [
 {
  bone1Name=thigh_r
  bone2Name=pelvis
  swing1LimitDegrees=45
  swing1Motion=Free
  swing2LimitDegrees=45
  swing2Motion=Free
  twistLimitDegrees=45
  twistMotion=Free
 }
 {
  bone1Name=calf_l
  bone2Name=thigh_l
  swing1LimitDegrees=45
  swing1Motion=Free
  swing2LimitDegrees=45
  swing2Motion=Free
  twistLimitDegrees=45
  twistMotion=Free
 }
 {
  bone1Name=calf_r
  bone2Name=thigh_r
  swing1LimitDegrees=45
  swing1Motion=Free
  swing2LimitDegrees=45
  swing2Motion=Free
  twistLimitDegrees=45
  twistMotion=Free
 }
 {
  bone1Name=thigh_l
  bone2Name=pelvis
  swing1LimitDegrees=45
  swing1Motion=Free
  swing2LimitDegrees=45
  swing2Motion=Free
  twistLimitDegrees=45
  twistMotion=Free
 }
]
